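# Additive unlit glow material for the BentBaton core demo mesh in the BeastGirlStormClaws content, flagged for skeletal mesh and mesh deformer use.
# Inverted vertex-color blue separates core from outer layer, blending emissive colors, mask settings and normal-based offset scale.
# A facing-angle sphere mask, distorted by a panning plasma texture, combines with a UV gradient and view-angle falloff.
# Offset math adds a target-position pull and a panning-noise sphere mask.
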Material /BeastGirlStormClawsGameplay/Assets/Meshes/Demo/demoMats/M_BentBaton_Core
graph {
 properties {
  bAutomaticallySetUsageInEditor=false
  bUsedWithMeshDeformer=true
  bUsedWithSkeletalMesh=true
  blendMode=BLEND_Additive
  shadingModel=MSM_Unlit
  usedShadingModels=Unlit
 }
 nodes [
  ActorPositionWS()
  n1 = VectorParameter()
  n2 = LocalPosition() {localOrigin=InstancePreSkinning}
  n3 = ComponentMask(None=TextureCoordinate()) {g=true}
  n4 = ScalarParameter() {defaultValue=1, parameterName=Param_3}
  n5 = LinearInterpolate(
   A=LinearInterpolate(A=ScalarParameter() {defaultValue=-2, parameterName=WPO Core Scale}, Alpha=Power(Base=@n3, Exp=@n4)) {constB=0}
   B=ScalarParameter() {parameterName=WPO scale}
   Alpha=OneMinus(None=VertexColor().B)
  )
  n6 = VertexNormalWS()
  n7 = VectorParameter() {parameterName=C, defaultValue=<rgba 1 0.5 0 0>}
  n8 = Panner(Coordinate=Multiply(A=TextureCoordinate(), B=ComponentMask(None=@n7.RGB) {g=true, r=true})) {speedY=1}
  n9 = Add(
   A=ComponentMask(None=TextureCoordinate()) {g=true}
   B=Multiply(
    A=TextureSample(UVs=@n8) {texture=&2ndInfectedMapLinear}.RGB
    B=ScalarParameter() {defaultValue=0.255, parameterName=Param_4}
   )
  )
  n10 = ScalarParameter() {defaultValue=1.0477, parameterName=Param_5}
  n11 = ScalarParameter() {defaultValue=1.164, parameterName=Param_6}
  n12 = ScalarParameter() {parameterName=Param_7}
  Add(
   A=Multiply(A=ScalarParameter() {parameterName=WPO}, B=Subtract(A=@n1.RGB, B=@n2.XYZ))
   B=Add(A=Multiply(A=@n5, B=@n6), B=SphereMask(A=@n9, B=@n10, Radius=@n11, Hardness=@n12) {attenuationRadius=0})
  )
  Fresnel() {exponent=1}
  MaterialFunctionCall(Emissive=LinearInterpolate(
    A=VectorParameter() {parameterName=Main Color, defaultValue=<rgba 10 6.6146 6.6146 0.5>}.RGB
    B=VectorParameter() {parameterName=Outer Color, defaultValue=<rgba 10 0 0.5 1>}.RGB
    Alpha=NamedRerouteUsage()
   )) {materialFunction=&MF_FX_EmissiveAdjustments}
  MaterialFunctionCall!(ID=ScalarParameter() {defaultValue=1}) {materialFunction=/Game/Packages/Fortress_SharedMaterials/Stencil_ID_Definitions/StencilMaskCompare}
  MaterialFunctionCall() {materialFunction=/Game/Effects/Niagara/MasterMaterials/Functions/InstanceBounds_0-1_UVW}
  Multiply(B=SceneTexture() {sceneTextureId=PPI_CustomDepth}.Color)
  n13 = DotProduct(A=PixelNormalWS(), B=CameraVectorWS())
  n14 = TextureCoordinate()
  n15 = ComponentMask(None=VectorParameter() {parameterName=Distortion TIling, defaultValue=<rgba 2 1 0 0>}.RGB) {g=true, r=true}
  n16 = Time()
  n17 = ScalarParameter() {defaultValue=1, parameterName=Time Scale}
  n18 = TextureSample(UVs=Panner(Coordinate=Multiply(A=@n14, B=@n15), Time=Multiply(A=@n16, B=@n17)) {speedY=1}) {samplerType=SAMPLERTYPE_Masks, texture=/Game/Weapons/WeaponSkins/Wraps/Textures/PlasmaCharge/PlasmaCharge_M}
  n19 = LinearInterpolate(
   A=ScalarParameter() {defaultValue=1, parameterName=Param_1}
   B=ScalarParameter() {defaultValue=1, parameterName=Distortion Scale}
   Alpha=NamedRerouteUsage()
  )
  n20 = NamedRerouteUsage()
  n21 = SphereMask(
   A=Add(A=Saturate(None=@n13), B=Multiply(A=@n18.B, B=@n19))
   B=LinearInterpolate(
    A=ScalarParameter() {defaultValue=1.6706, parameterName=Mask Core}
    B=ScalarParameter() {defaultValue=2.8181, parameterName=Mask Outer}
    Alpha=@n20
   )
   Radius=LinearInterpolate(
    A=ScalarParameter() {defaultValue=0.9, parameterName=Mask Radius Core}
    B=ScalarParameter() {defaultValue=2, parameterName=Mask Radius}
    Alpha=@n20
   )
   Hardness=ScalarParameter() {defaultValue=-1, parameterName=Mask Hardness}
  ) {hardnessPercent=1}
  n22 = ComponentMask(None=TextureCoordinate()) {g=true}
  n23 = ScalarParameter() {defaultValue=1, parameterName=Param_2}
  n24 = LinearInterpolate(
   A=ScalarParameter() {defaultValue=0.6, parameterName=Gradient}
   B=ScalarParameter() {defaultValue=0.65, parameterName=Gradient Outter}
   Alpha=OneMinus(None=VertexColor().B)
  )
  n25 = ScalarParameter() {parameterName=Param_10}
  n26 = DotProduct(A=CameraVectorWS(), B=VertexNormalWS())
  Multiply(
   A=Multiply(A=@n21, B=SphereMask(A=@n22, B=@n23, Radius=@n24, Hardness=@n25) {hardnessPercent=0})
   B=Power(Base=Saturate(None=@n26), Exp=ScalarParameter() {defaultValue=1, parameterName=Param_8})
  )
  Multiply(B=@n21)
  NamedRerouteDeclaration(None=OneMinus(None=VertexColor().B)) {name=Vertex Color, nodeColor=<rgba 1 0 0.4219 1>}
  ObjectPositionWS()
  OneMinus()
  OneMinus(None=SceneTexture() {sceneTextureId=PPI_CustomStencil}.Color)
  ScalarParameter() {defaultValue=2, parameterName=Distortion Tiling}
  TransformPosition() {transformType=TRANSFORMPOSSOURCE_World}
  VertexColor()
 ]
}
