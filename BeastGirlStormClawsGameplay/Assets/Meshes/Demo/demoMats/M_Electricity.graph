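# M_Electricity is an unlit, two-sided translucent material for animated electricity effects in the demo assets, flagged for skeletal mesh use.
# World position offset combines SplineThicken width with normal-direction warble from a panning texture, both modulated through named reroutes.
# Reroutes also define a looping Lifetime staggered by the third UV channel, a SphereMask time fade, and a gradient that peaks midway along the second UV channel.
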Material /BeastGirlStormClawsGameplay/Assets/Meshes/Demo/demoMats/M_Electricity
graph {
 properties {
  bAutomaticallySetUsageInEditor=false
  bEnableResponsiveAA=true
  bOutputTranslucentVelocity=true
  bUsedWithMeshDeformer=true
  bUsedWithSkeletalMesh=true
  blendMode=BLEND_Translucent
  shadingModel=MSM_Unlit
  twoSided=true
  usedShadingModels=Unlit
 }
 nodes [
  Add(B=ComponentMask(None=TextureCoordinate() {coordinateIndex=1}) {g=true})
  ComponentMask(None=TextureCoordinate()) {g=true, r=true}
  Constant()
  Multiply()
  n1 = Time()
  n2 = ComponentMask(None=TextureCoordinate() {coordinateIndex=2}) {g=true}
  n3 = ScalarParameter() {defaultValue=11, parameterName=Lifetime Low}
  n4 = ScalarParameter() {defaultValue=1, parameterName=Lifetime High}
  NamedRerouteDeclaration(None=Fmod(A=Add(A=@n1, B=@n2), B=LinearInterpolate(A=@n3, B=@n4, Alpha=@n2) {constA=1})) {name=Lifetime, nodeColor=<rgba 0 1 0.0156 1>}
  n5 = Time() {period=1}
  n6 = ScalarParameter() {defaultValue=1, parameterName=Time Loop}
  NamedRerouteDeclaration(None=SphereMask(
    A=Fmod(A=@n5, B=@n6)
    B=Constant() {r=0.5}
    Radius=ScalarParameter() {defaultValue=2, parameterName=Radius}
    Hardness=ScalarParameter() {parameterName=Hardness}
   )) {name=FadeInOut, nodeColor=<rgba 0.5469 1 0 1>}
  n7 = LinearInterpolate(
   A=ScalarParameter() {parameterName=Spline Minimum Size}
   B=LinearInterpolate(B=ScalarParameter() {parameterName=Spline Thickness}, Alpha=NamedRerouteUsage())
   Alpha=NamedRerouteUsage()
  )
  n8 = TextureCoordinate() {coordinateIndex=1}
  n9 = MaterialFunctionCall(WidthBase=@n7, WidthTip=@n7, WorldPosition=WorldPosition().XYZ, UVs for Projection=@n8, UVs for Thickness=@n8) {materialFunction=&SplineThicken}
  n10 = ScalarParameter() {defaultValue=1, parameterName=Debug WPO Spline Scale}
  n11 = NamedRerouteUsage()
  n12 = Panner(Coordinate=TextureCoordinate() {coordinateIndex=1}) {speedY=1}
  n13 = Multiply(
   A=Multiply(A=TextureSample(UVs=@n12) {texture=&2ndInfectedMapLinear}.RGB, B=VertexNormalWS())
   B=LinearInterpolate(B=ScalarParameter() {defaultValue=40, parameterName=Warble Size}, Alpha=NamedRerouteUsage())
  )
  NamedRerouteDeclaration(None=Add(A=Multiply(A=@n9.WorldPositionOffset, B=@n10), B=Multiply(A=@n11, B=@n13))) {name=WPO, nodeColor=<rgba 1 0.4922 0 1>}
  n14 = ComponentMask(None=TextureCoordinate() {coordinateIndex=1}) {g=true}
  NamedRerouteDeclaration(None=Multiply(A=OneMinus(None=@n14), B=@n14)) {name=Top and Bottom Gradient, nodeColor=<rgba 0 1 0.4844 1>}
  NamedRerouteUsage()
  NamedRerouteUsage()
  NamedRerouteUsage()
  NamedRerouteUsage()
  NamedRerouteUsage()
  NamedRerouteUsage()
  ScalarParameter() {parameterName=Param_7}
  VectorParameter() {parameterName=Base Color, defaultValue=<rgba 8 0 0.1333 1>}
 ]
}
