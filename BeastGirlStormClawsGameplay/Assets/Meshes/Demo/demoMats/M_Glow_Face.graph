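# Unlit, translucent, two-sided glow material for skeletal meshes, named for a face and sampling a face-accessory FX texture.
# Time-scrolled local-space Voronoi noise perturbs a vertical sphere mask that blends red to black.
# The inverted mask multiplies that FX texture's green channel, sampled with noise-distorted UVs.
# Further branches build a normal-based, height-weighted offset and an electricity-noise radial mask; no output connections appear.
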Material /BeastGirlStormClawsGameplay/Assets/Meshes/Demo/demoMats/M_Glow_Face
graph {
 properties {
  bAutomaticallySetUsageInEditor=false
  bUsedWithMeshDeformer=true
  bUsedWithSkeletalMesh=true
  blendMode=BLEND_Translucent
  shadingModel=MSM_Unlit
  twoSided=true
  usedShadingModels=Unlit
 }
 nodes [
  n1 = LocalPosition()
  n2 = Multiply(A=VectorParameter() {parameterName=Param_5, defaultValue=<rgba 0 50 0 0>}.RGB, B=Time())
  n3 = TextureSample(UVs=Multiply(A=Add(A=@n1.XYZ, B=@n2), B=VectorParameter() {parameterName=Param_2, defaultValue=<rgba 0.1 0.0227 0.1 0>}.RGB)) {samplerType=SAMPLERTYPE_LinearColor, texture=/Game/Effects/FogVolume/VT_Voronoi_LOW}
  n4 = Multiply(A=@n3.RGB, B=ScalarParameter() {defaultValue=0.07, parameterName=Param_19})
  n5 = Normalize(VectorInput=LocalPosition().XYZ)
  n6 = SphereMask(
   A=Add(A=@n4, B=ComponentMask(None=@n5) {g=true})
   B=ScalarParameter() {defaultValue=0.0107, parameterName=Param_14}
   Radius=ScalarParameter() {defaultValue=0.1317, parameterName=Param_15}
   Hardness=ScalarParameter() {defaultValue=0.75, parameterName=Param_17}
  )
  LinearInterpolate(
   A=VectorParameter() {parameterName=Param_4, defaultValue=<rgba 2 0 0 0>}.RGB
   B=VectorParameter() {parameterName=Param_3, defaultValue=<rgba 0 0 0 1>}.RGB
   Alpha=@n6
  )
  n7 = Multiply(A=ComponentMask(None=@n3.RGB) {g=true, r=true}, B=ScalarParameter() {defaultValue=0.01, parameterName=Param_2})
  n8 = TextureCoordinate()
  Multiply(
   A=Saturate(None=OneMinus(None=@n6))
   B=TextureSample(UVs=Add(A=@n7, B=@n8)) {texture=&/T_FaceAcc_BentBaton_FX}.G
  )
  OneMinus()
  n9 = VertexNormalWS()
  n10 = ScalarParameter() {defaultValue=20}
  n11 = VectorParameter() {parameterName=WPO Direction}
  n12 = LocalPosition()
  n13 = Power(Base=ComponentMask(None=Normalize(VectorInput=@n12.XYZ)) {g=true}, Exp=ScalarParameter() {defaultValue=1.5, parameterName=Param_1})
  Reroute(None=Multiply(A=Multiply(A=@n4, B=@n9), B=LinearInterpolate(A=@n10, B=@n11.RGB, Alpha=@n13)))
  n14 = ScalarParameter() {defaultValue=4, parameterName=Param_12}
  n15 = VectorParameter() {parameterName=Scale, defaultValue=<rgba 1 1 0.5 0>}
  n16 = Panner(Coordinate=Multiply(A=TextureCoordinate(), B=ComponentMask(None=@n15.RGB) {g=true, r=true})) {speedY=1}
  n17 = Add(
   A=TextureCoordinate()
   B=Multiply(
    A=TextureSample(UVs=@n16) {samplerType=SAMPLERTYPE_Masks, texture=/Game/Characters/Player/Male/Medium/Bodies/M_MED_Storm_Soldier/Textures/ElectricityNoise_M}.G
    B=ScalarParameter() {defaultValue=1, parameterName=Param_5}
   )
  )
  n18 = ComponentMask(None=VectorParameter() {defaultValue=<rgba 0 -2.1739 0 0>}.RGB) {g=true, r=true}
  n19 = Multiply(A=Subtract(A=Add(A=@n17, B=@n18)) {constB=0.5}) {constB=0.5}
  n20 = Multiply(A=OneMinus(None=Length(None=@n19) {is=class MaterialExpressionLength}), B=ScalarParameter() {defaultValue=1, parameterName=Param_4})
  n21 = TextureCoordinate()
  SphereMask(
   A=Add(A=Multiply(A=@n14, B=@n20), B=ComponentMask(None=@n21) {g=true})
   B=ScalarParameter() {defaultValue=1.196, parameterName=Param_9}
   Radius=ScalarParameter() {defaultValue=0.8373, parameterName=Param_11}
   Hardness=ScalarParameter() {defaultValue=0.9637, parameterName=Param_10}
  ) {attenuationRadius=1}
  VectorParameter() {parameterName=Param_1, defaultValue=<rgba -500 -500 -500 0>}
 ]
}
